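; Looping gameplay cue that plays the transition effect while a vehicle's cosmetics swap.
; On start it attaches loading Niagara and start audio, then sets a custom primitive data float on body and attached meshes.
; Removal flags the VFX loaded, plays end audio, fades the loop, and animates that float by timeline before deactivating the VFX.
; A material-swap function and a wheel-check event with an empty loop go uncalled here.
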(event EventOnLoopingStart(EitherVFX-preferthisversion)
  (params (MyTarget :type Actor%20Object%20Reference) (Parameters :type Gameplay%20Cue%20Parameters%20Structure%20%28by%20ref%29) (ParticleComponents :type Array%20of%20FXSystem%20Component%20Object%20References) (AudioComponents :type Array%20of%20Audio%20Component%20Object%20References))
  (bind _asfort_vehicle_interface (Utilities|Casting|CastToFortVehicleInterface MyTarget)
    (:then
      (Development|PrintString 0 (Utilities|String|Append "Looping Start - Get Skeletal Mesh Component = " (Utilities|GetDisplayName (Class|FortVehicleInterface|GetSkeletalMeshComponent _asfort_vehicle_interface))) (Variables|Default|GetDEBUG) (Variables|Default|GetDEBUG) "(R=0.305042,G=1.000000,B=0.000000,A=1.000000)" 2e+01)
      (bind _returnvalue (Niagara|SpawnSystemAttached "/Game/Effects/Fort_Effects/Gameplay/Vehicles/NS_Vehicle_Transition_Loading_v4.NS_Vehicle_Transition_Loading_v4" (Class|FortVehicleInterface|GetSkeletalMeshComponent _asfort_vehicle_interface) "Center" "0, 0, 0" "0, 0, 0" "KeepRelativeOffset" true))
      (Variables|Default|SetLoopingVFX _returnvalue)
      (bind _returnvalue_2 (Audio|SpawnSoundAttached "/VehicleCosmetics/Audio/CosmeticSwap/Veh_CustomizationSwap_Start_Cue.Veh_CustomizationSwap_Start_Cue" (Class|FortVehicleInterface|GetSkeletalMeshComponent _asfort_vehicle_interface)))
      (Variables|Default|SetLoopingSound _returnvalue_2)
      (Rendering|Material|SetCustomPrimitiveDataFloat (Class|FortVehicleInterface|GetSkeletalMeshComponent _asfort_vehicle_interface) 13 1.0)
      (Rendering|Material|SetCustomPrimitiveDataFloat (Class|FortVehicleInterface|GetAttachedSkeletalMeshes _asfort_vehicle_interface) 13 1.0))
    (:CastFailed
      (Development|PrintString 0 (Utilities|String|Append "Looping Start - CAST FAILED - " "Target = " (Utilities|GetDisplayName MyTarget)) (Variables|Default|GetDEBUG) (Variables|Default|GetDEBUG) "(R=1.000000,G=0.000000,B=0.000000,A=1.000000)" 1e+01))))

(event Custom|KeepCheckingForWheels
  (bind _asfort_vehicle_interface (Utilities|Casting|CastToFortVehicleInterface (Variables|Default|GetLocalFortVehicleInterface))
    (:then
      (for _array_element (Class|FortVehicleInterface|GetAttachedSkeletalMeshes _asfort_vehicle_interface)
        _))
    (:CastFailed
      (Utilities|Time|ClearandInvalidateTimerbyHandle 0 (Variables|Default|GetKeepCheckingTimer)))))

(event EventOnRemoval(EitherVFX-preferthisversion)
  (params (MyTarget :type Actor%20Object%20Reference) (Parameters :type Gameplay%20Cue%20Parameters%20Structure%20%28by%20ref%29) (ParticleComponents :type Array%20of%20FXSystem%20Component%20Object%20References) (AudioComponents :type Array%20of%20Audio%20Component%20Object%20References) (BurstCameraShakeInstance :type Camera%20Shake%20Base%20Object%20Reference) (BurstDecalInstance :type Decal%20Actor%20Object%20Reference))
  (bind _loopingvfx (Variables|Default|GetLoopingVFX))
  (bind _asfort_vehicle_interface (Utilities|Casting|CastToFortVehicleInterface MyTarget)
    (:then
      (Development|PrintString 0 (Utilities|String|Append "Removal- Get Skeletal Mesh Component = " (Utilities|GetDisplayName (Class|FortVehicleInterface|GetSkeletalMeshComponent _asfort_vehicle_interface))) (Variables|Default|GetDEBUG) (Variables|Default|GetDEBUG) "(R=1.000000,G=0.696089,B=0.000000,A=1.000000)" 2e+01)
      (Effects|Components|ParticleSystem|SetBoolParameter _loopingvfx "Loaded" true)
      (Audio|SpawnSoundAttached "/VehicleCosmetics/Audio/CosmeticSwap/Veh_CustomizationSwap_End_Cue.Veh_CustomizationSwap_End_Cue" (Class|FortVehicleInterface|GetSkeletalMeshComponent _asfort_vehicle_interface)) ; 'End' sound
      (Audio|Components|Audio|FadeOut (Variables|Default|GetLoopingSound) 0.7 0.0 "Linear" :!class /Script/Engine.AudioComponent)
      (bind (_direction _car_sweep_controller) (Timeline|Sweep%20Car%20to%20Normal Play)
        (:Update
          (Rendering|Material|SetCustomPrimitiveDataFloat (Class|FortVehicleInterface|GetSkeletalMeshComponent _asfort_vehicle_interface) 13 _car_sweep_controller)
          (Rendering|Material|SetCustomPrimitiveDataFloat (Class|FortVehicleInterface|GetAttachedSkeletalMeshes _asfort_vehicle_interface) 13 _car_sweep_controller))
        (:Finished
          (Components|Activation|Deactivate _loopingvfx))))
    (:CastFailed
      (Development|PrintString 0 (Utilities|String|Append "Removal - CAST FAILED - " "Target = " (Utilities|GetDisplayName MyTarget)) (Variables|Default|GetDEBUG) (Variables|Default|GetDEBUG) "(R=1.000000,G=0.000000,B=0.000000,A=1.000000)" 1e+01))))

(fn MaterialSwitch
  (params (Restore%20Original :type Boolean) (Vehicle :type Fort%20Vehicle%20Interface%20Interface))
  (if Restore%20Original
    (for (_ _array_index) (Rendering|Material|GetMaterials (Class|FortVehicleInterface|GetSkeletalMeshComponent (Variables|GetVehicle)))
      (Rendering|Material|SetMaterial (Class|FortVehicleInterface|GetSkeletalMeshComponent (Variables|GetVehicle)) _array_index (Utilities|Array|Get(acopy) (Variables|Default|GetOG%20Vehicle%20Mats) _array_index)))
    (else
      (Variables|Default|SetOG%20Vehicle%20Mats (Rendering|Material|GetMaterials (Class|FortVehicleInterface|GetSkeletalMeshComponent (Variables|GetVehicle))))
      (for (_ _array_index_2) (Rendering|Material|GetMaterials (Class|FortVehicleInterface|GetSkeletalMeshComponent (Variables|GetVehicle)))
        (Rendering|Material|SetMaterial (Class|FortVehicleInterface|GetSkeletalMeshComponent (Variables|GetVehicle)) _array_index_2 (Variables|Default|GetVehicle_DeRez_Material))))))

(fn ConstructionScript)
